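; POI claim banner actor that tracks per-team capture progress and, at full progress, declares a winning squad.
; The server spawns player banners on a ring and applies a progress-tracker GameplayEffect to participants, restored for returning teams via bounds events.
; Replicated FlagProgress lifts the ring toward the pole top and drives a hologram material parameter collection.
; Capture shows winners on banners, updates the minimap icon and sends quest events.
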(event EventBeginPlay
  (bind (_relativelocation_x _relativelocation_y _relativelocation_z) (Math|Vector|BreakVector (Variables|/Script/Engine.SceneComponent|GetRelativeLocation_X (Variables|Default|GetBannerRing))))
  (Networking|SwitchHasAuthority
    (:Authority
      (Variables|Default|SetBannerRingInitialZ _relativelocation_z)
      (CallFunction|SpawnBanners))
    (:Remote
      (CallFunction|InitializeProgressWidget)
      (CallFunction|UpdateFlagProgress))))

(event Custom|Initialize
  (bind _bounds (Variables|Default|GetBounds))
  (Networking|SwitchHasAuthority
    (:Authority
      (Default|BindEventtoOnActorEnteredBounds _bounds (EventDispatchers|CreateEvent :!event OnActorEnteredBounds))
      (Default|BindEventtoOnActorExitBounds _bounds (EventDispatchers|CreateEvent :!event OnActorExitBounds)))))

(event Collision|EventActorBeginOverlap
  (params (OtherActor :type Actor%20Object%20Reference)))

(event EventTick
  (params (DeltaSeconds :type Float%20%28single-precision%29)))

(event Custom|POIClaimResetAllProgress
  (Networking|FlushNetDormancy)
  (Progress|ResetAllProgress)
  (Variables|Default|SetFlagProgress 0.0)
  (MiniMapData|SetMiniMapIcon (Variables|Default|GetFortMiniMapComponent) (Variables|MapIcon|GetMapIcon_Uncontested))
  (Rendering|Material|SetScalarParameterValue "/Game/Environments/TechArt/Hologram/MPC_Hologram.MPC_Hologram" "ColorSwitch" :!class /Script/Engine.KismetMaterialLibrary)
  (Rendering|Material|SetScalarParameterValue "/Game/Environments/TechArt/Hologram/MPC_Hologram.MPC_Hologram" "Captured" :!class /Script/Engine.KismetMaterialLibrary)
  (Rendering|Material|SetScalarParameterValue "/Game/Environments/TechArt/Hologram/MPC_Hologram.MPC_Hologram" "Contested" :!class /Script/Engine.KismetMaterialLibrary))

(event Custom|InitializeProgressWidget
  (EventDispatchers|BindEventtoOnClampStateChangedEvent (Variables|Default|GetProgressWidgetComponent) (delegate Custom|OnWidgetComponentClampStateChanged2)))

(event Custom|OnWidgetComponentClampStateChanged2
  (params (ClampToScreenDirection :type EClampToScreenDirection%20Enum :default "Left"))
  (bind _aswbp_poiwin_progress_indicator (Utilities|Casting|CastToWBP_POIWin_ProgressIndicator (UserInterface|GetWidget (Variables|Default|GetProgressWidgetComponent))))
  (BindEvents|OnClampStateChanged _aswbp_poiwin_progress_indicator ClampToScreenDirection))

(event EventEndPlay
  (params (EndPlayReason :type EEndPlayReason%20Enum :default "Destroyed"))
  ;[0.07292 0.04481 0.003781] EndPlay: Destroy the spawned banners.
  (|CallParentFunction EndPlayReason)
  (Networking|SwitchHasAuthority
    (:Authority
      (Actor|DestroyActor (Variables|Default|GetBanners)))))

(fn AddProgress
  (params (Team :type Integer) (Progress :type Float%20%28double-precision%29) (bOverrideProgress :type Boolean))
  (bind _progressteammap (Variables|Default|GetProgressTeamMap))
  (bind (_value _returnvalue) (Utilities|Map|Find _progressteammap Team))
  (bind _boverrideprogress (Variables|GetbOverrideProgress))
  (bind _progress (Variables|GetProgress))
  (bind _currentprogress (LocalVariables|GetCurrentProgress))
  (bind _returnvalue_2 (+ _currentprogress _progress))
  (bind _progress_2 (Variables|GetProgress))
  (bind _returnvalue_3 (Math|Float|SelectFloat _progress_2 _returnvalue_2 _boverrideprogress))
  (bind _team (Variables|Team|GetTeam))
  (bind _team_2 (Variables|Team|GetTeam))
  (bind _progress_3 (Variables|GetProgress))
  (if (not (Variables|Default|GetbIsRaised))
    (LocalVariables|SetCurrentProgress _value)
    (if _returnvalue
      (Utilities|Map|Add (Variables|Default|GetProgressTeamMap) (Variables|Team|GetTeam) _returnvalue_3)
      (LocalVariables|SetNewProgress _returnvalue_3)
      (else
        (Utilities|Map|Add (Variables|Default|GetProgressTeamMap) (Variables|Team|GetTeam) _progress_3)
        (LocalVariables|SetNewProgress _progress_3)))
    (Variables|Default|SetFlagProgress (LocalVariables|GetNewProgress))
    (Variables|Default|SetCurrentTeam _team_2)
    (Default|CallOnProgressUpdated _team_2 (LocalVariables|GetNewProgress))
    (if (>= (LocalVariables|GetNewProgress) 1.0)
      (Progress|HandleBannerRaised _team)
      (Variables|Default|SetbIsRaised true)
      (Default|CallOnBannerRaised _team (Variables|Default|GetWinner%20Pawns) self))))

(fn CreateBannerForPlayer
  (params (Pawn :type Fort%20Pawn%20Object%20Reference) (Index :type Integer))
  (Class|BPOIClaimPlayerBanner|SetRepresentedPlayer (Utilities|Array|Get(acopy) (Variables|Default|GetBanners) Index) Pawn))

(fn DisplayWinners
  (params (WinningPawns :type Array%20of%20Fort%20Pawn%20Object%20References))
  (bind _winningpawns (Variables|GetWinningPawns))
  (bind _output (Utilities|Array|Get(acopy) _winningpawns 1))
  (bind _winningpawns_2 (Variables|GetWinningPawns))
  (bind _output_2 (Utilities|Array|Get(acopy) _winningpawns_2))
  (bind _winningpawns_3 (Variables|GetWinningPawns))
  (bind _output_3 (Utilities|Array|Get(acopy) _winningpawns_3))
  (switch int (Utilities|Array|Length WinningPawns)
    (:1
      (for _index (range 4)
        (CallFunction|CreateBannerforPlayer (Utilities|Array|Get(acopy) (Variables|GetWinningPawns)) _index)))
    (:2
      (CallFunction|CreateBannerforPlayer _output_2)
      (CallFunction|CreateBannerforPlayer _output_2 2)
      (CallFunction|CreateBannerforPlayer _output 1)
      (CallFunction|CreateBannerforPlayer _output 3))
    (:3
      (CallFunction|CreateBannerforPlayer _output_3)
      (CallFunction|CreateBannerforPlayer _output_3 2)
      (CallFunction|CreateBannerforPlayer (Utilities|Array|Get(acopy) (Variables|GetWinningPawns) 1) 1)
      (CallFunction|CreateBannerforPlayer (Utilities|Array|Get(acopy) (Variables|GetWinningPawns) 2) 3))
    (:4
      (for (_array_element _array_index) (Variables|GetWinningPawns)
        (CallFunction|CreateBannerforPlayer _array_element _array_index)))))

(fn GetCurrentTeamProgress
  (bind (_value _returnvalue) (Utilities|Map|Find (Variables|Default|GetProgressTeamMap) (Variables|Default|GetCurrentTeam)))
  (return _returnvalue _value))

(fn GetDistanceBetweenBanners
  (return (/ (Variables|Default|GetTotalBannersWidth) (Variables|Default|GetNumWinners))))

(fn GetFirstParticipatingActorFromWinningTeam
  (bind _array_element (Utilities|Array|ForEachLoopwithBreak (Variables|Default|GetParticipatingActors))
    (:LoopBody
      (if (== (Teams|TeamtoIndex (Team|GetActorTeam _array_element)) (Variables|Default|GetWinningTeam))
        (LocalVariables|SetOutActor _array_element)))
    (:Completed
      _))
  (return (LocalVariables|GetOutActor)))

(fn GetFirstPlayerBannerLocation
  (return (+ (Transformation|GetWorldLocation (Variables|Default|GetPoleTop)) (Math|Vector|RotateVector (Math|Vector|MakeVector 0.0 (* (Variables|Default|GetTotalBannersWidth) -0.5)) (Transformation|GetActorRotation)))))

(fn GetFirstProgressSourceAsCapturePoint
  (bind _progresssources (Variables|Default|GetProgressSources))
  (if (Utilities|Array|IsValidIndex _progresssources)
    (bind _asbga_poibanner_capture_point (Utilities|Casting|CastToBGA_POIBanner_CapturePoint (Utilities|Array|Get(acopy) _progresssources)))
    (return _asbga_poibanner_capture_point)))

(fn GetPlayerBannerLocation
  (params (Index :type Integer))
  (return (+ (CallFunction|GetFirstPlayerBannerLocation) (Math|Vector|RotateVector (Math|Vector|MakeVector 0.0 (* (CallFunction|GetDistanceBetweenBanners) Index)) (Transformation|GetActorRotation)))))

(fn GetShowPoiToastOnCapture
  (return (Variables|Default|GetbShowPOIToastOnCapture)))

(fn GetTeamProgress
  (params (TeamIndex :type Integer))
  (bind (_value _returnvalue) (Utilities|Map|Find (Variables|Default|GetProgressTeamMap) TeamIndex))
  (return _returnvalue _value))

(fn GetWinners
  (bind _returnvalue (CallFunction|GetFirstParticipatingActorfromWinningTeam))
  (bind _asfort_player_pawn (Utilities|Casting|CastToFortPlayerPawn _returnvalue))
  (bind _asfort_player_state_athena (Utilities|Casting|CastToFortPlayerStateAthena (Variables|/Script/Engine.Pawn|GetPlayerState _asfort_player_pawn)))
  (bind _returnvalue_2 (Class|FortGameStateAthena|GetSquadMemberPlayerStates (Fortnite|Athena|GameState|GetGameStateAthena) _asfort_player_state_athena true))
  (for _array_element _returnvalue_2
    (Utilities|IsValid _array_element
      (:Is%20Valid
        (Utilities|Array|Add (LocalVariables|GetOutPawns) (PlayerState|GetCurrentPawn _array_element)))))
  (return (LocalVariables|GetOutPawns)))

(fn HandleBannerRaised
  (params (Team :type Integer))
  (MiniMapData|SetMiniMapIcon (Variables|Default|GetFortMiniMapComponent) (Variables|MapIcon|GetMapIcon_Claimed))
  (Variables|Default|SetWinningTeam Team)
  (Utilities|FlowControl|Sequence
    (:then_0
      (for (_array_element _array_index) (Variables|Default|GetBanners)
        (Class|BPOIClaimPlayerBanner|OnCaptured _array_element _array_index)))
    (:then_1
      (bind _winnerpawns (CallFunction|GetWinners))
      (Variables|Default|SetWinner%20Pawns _winnerpawns)
      (CallFunction|DisplayWinners _winnerpawns)
      ;[1 1 1] Send Quest Event
      (for _array_element_2 _winnerpawns
        (Variables|Default|SetFort%20Player%20Ref _array_element_2)
        (Utilities|IsValid (Fortnite|GetFortPlayerControllerfromActor _array_element_2)
          (:Is%20Valid
            (Quest|SendComplexCustomStatEvent (Quests|GetQuestManager (Fortnite|GetFortPlayerControllerfromActor _array_element_2) "Invalid") 0 (Variables|Default|GetTag%20Quest%20POI%20Claimed) "(GameplayTags=)" 1)
            ;[1 1 1] Send Custom QS2 Progress
            (bind _result (CustomVerb|AddTagsforSubject (CustomVerb|MakeCustomVerbMessage (Fortnite|GetFortPlayerControllerfromActor _array_element_2) "(TagName=\"CustomVerb.POIClaimed\")") "Athena.Quest.POIClaimed" (GameplayTags|MakeGameplayTagContainerfromTag "(TagName=\"Athena.Quest.POIClaimed\")")))
            (CustomVerb|BroadcastCustomVerbMessage _result))) ; No player controller for bots
      ))))

(fn InformBannerState
  (params (State :type EPOIBannerState%20Enum :default "NewEnumerator0"))
  (Development|PrintString (Utilities|String|EnumtoString State))
  (for _array_element (Variables|Default|GetBanners)
    (Class|BPOIClaimPlayerBanner|SetVisualsfromState _array_element State)))

(fn NotifyProgressHalted
  (params (Team :type Integer))
  (Default|CallOnProgressHalted Team)
  (Networking|FlushNetDormancy)
  (MiniMapData|SetMiniMapIcon (Variables|Default|GetFortMiniMapComponent) (Variables|MapIcon|GetMapIcon_Contested)))

(fn NotifyProgressStarted
  (params (Team :type Integer))
  (Default|CallOnProgressStarted Team)
  (Networking|FlushNetDormancy)
  (MiniMapData|SetMiniMapIcon (Variables|Default|GetFortMiniMapComponent) (Variables|MapIcon|GetMapIcon_Uncontested)))

(fn OnActorEnteredBounds
  (params (Actor :type Actor%20Object%20Reference))
  ;[1 1 1] Triggered when a pawn enters the overall bounds of the POI (note: this is not the same as the capture point zone). If the actor's team already has progress, give them the progress tracking ability. This is meant to handle players who have started claiming a POI, leave the area, then return
  (bind _keys (Utilities|Map|Keys (Variables|Default|GetProgressTeamMap)))
  (if (Utilities|Array|ContainsItem _keys (Teams|TeamtoIndex (Team|GetActorTeam Actor)))
    (Progress|TryAddParticipant Actor)))

(fn OnActorExitBounds
  (params (Actor :type Actor%20Object%20Reference))
  (CallFunction|RemoveParticipant Actor))

(fn OnRep_FlagProgress
  (CallFunction|UpdateFlagProgress)
  (Default|CallOnFlagProgressUpdated (Variables|Default|GetFlagProgress)))

(fn RemoveParticipant
  (params (Actor :type Actor%20Object%20Reference))
  (bind _participatingactors (Variables|Default|GetParticipatingActors))
  (if (Utilities|Array|ContainsItem _participatingactors Actor)
    (Utilities|Array|RemoveItem _participatingactors Actor)
    (GameplayEffects|RemoveActiveGameplayEffectbySourceEffect (Ability|GetAbilitySystemComponent Actor) (Variables|Default|GetGE_ProgressTracker) 0 -1)))

(fn ResetAllProgress
  (bind _keys (Utilities|Map|Keys (Variables|Default|GetProgressTeamMap)))
  (for _array_element _keys
    (Progress|ResetTeamProgress _array_element)
    (Default|CallOnProgressUpdated _array_element)))

(fn ResetAllProgressWithExclusion
  (params (ExcludeTeams :type Boolean) (TeamsToExclude :type Array%20of%20Integers))
  (bind _keys (Utilities|Map|Keys (Variables|Default|GetProgressTeamMap)))
  (for _array_element _keys
    (if (not (Utilities|Array|ContainsItem (Variables|GetTeamsToExclude) _array_element))
      (Progress|ResetTeamProgress _array_element))))

(fn ResetTeamProgress
  (params (Team :type Integer))
  (Utilities|Map|Add (Variables|Default|GetProgressTeamMap) Team))

(fn SpawnBanners
  (if (Variables|Default|GetbSpawnBanners)
    (for _array_element (Variables|Default|GetBannerInitialTransforms)
      (bind _banneractor (CallFunction|SpawnPlayerBanner (Math|Transform|MakeTransform (Math|Transform|TransformLocation (Transformation|GetActorTransform) (+ (Variables|/Script/Engine.SceneComponent|GetRelativeLocation (Variables|Default|GetBannerRing)) (.location _array_element))) (Math|Transform|TransformRotation (Transformation|GetActorTransform) (.rotation _array_element)) (.scale _array_element))))
      (Utilities|Array|Add (Variables|Default|GetBanners) _banneractor)
      (Transformation|AttachActorToComponent _banneractor (Variables|Default|GetBannerRing) "None" "KeepWorld" "KeepWorld" "KeepWorld" true))))

(fn SpawnPlayerBanner
  (params (SpawnTransform :type Transform))
  (bind _returnvalue (Game|SpawnActorfromClass (Variables|Default|GetPlayerBannerClass) SpawnTransform "Undefined" "SelectDefaultAtRuntime" 0 0))
  (bind _asb_poiclaim_player_banner (Utilities|Casting|CastToB_POIClaim_PlayerBanner _returnvalue))
  (return _asb_poiclaim_player_banner))

(fn TryAddParticipant
  (params (Actor :type Actor%20Object%20Reference))
  (if (not (Utilities|Array|ContainsItem (Variables|Default|GetParticipatingActors) Actor))
    (Utilities|Array|Add (Variables|Default|GetParticipatingActors) (Variables|GetActor))
    (GameplayEffects|ApplyGameplayEffectToSelf (Ability|GetAbilitySystemComponent (Variables|GetActor)) (Variables|Default|GetGE_ProgressTracker) 0.0 (Class|AthenaFunctionLibrary|MakeSimpleFortEffectContext self))))

(fn UpdateFlagProgress
  (bind _flagprogress (Variables|Default|GetFlagProgress))
  (bind (_relativelocation_x _relativelocation_y _relativelocation_z) (Math|Vector|BreakVector (Variables|/Script/Engine.SceneComponent|GetRelativeLocation_X (Variables|Default|GetPoleTop))))
  (Utilities|FlowControl|Sequence
    (:then_0
      (Transformation|SetRelativeLocation (Variables|Default|GetBannerRing) (Math|Vector|MakeVector 0.0 0.0 (Math|Float|Lerp (Variables|Default|GetBannerRingInitialZ) _relativelocation_z (Variables|Default|GetFlagProgress))) false false)
      ;[1 1 1] moved to /Script/Engine.Blueprint'/POIBanner/Actors/CapturePoint/BGA_POIBanner_CapturePoint.BGA_POIBanner_CapturePoint' to fix possible slow replication
      (Rendering|Material|SetScalarParameterValue "/Game/Environments/TechArt/Hologram/MPC_Hologram.MPC_Hologram" "ProgressNormalized" (Variables|Default|GetFlagProgress) :!class /Script/Engine.KismetMaterialLibrary)
      (if (> _flagprogress 0)
        (Rendering|Material|SetScalarParameterValue "/Game/Environments/TechArt/Hologram/MPC_Hologram.MPC_Hologram" "ScreensEnabled" 1.0 :!class /Script/Engine.KismetMaterialLibrary)
        (elif (<= _flagprogress 0)
          (Rendering|Material|SetScalarParameterValue "/Game/Environments/TechArt/Hologram/MPC_Hologram.MPC_Hologram" "ScreensEnabled" :!class /Script/Engine.KismetMaterialLibrary))))))

(fn ConstructionScript)
