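; Parent Blueprint for actors that drive team progress on a POI banner, presumably extended by children that supply the progress logic.
; On BeginPlay it flushes net dormancy so the spawn-set BannerRef replicates, then binds to the banner's OnBannerRaised event.
; Validity-checked functions forward progress updates and all-team, per-team or exclusion-based resets to the banner.
; The raised handler only prints the team, leaving cleanup to child overrides; the BannerRef OnRep is empty.
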(event EventBeginPlay
  (bind _bannerref (Variables|Default|GetBannerRef))
  (Networking|FlushNetDormancy) ; Ensure BannerRef (which is set on spawn) replicates.
  (Utilities|IsValid _bannerref
    (:Is%20Valid
      (Default|BindEventtoOnBannerRaised _bannerref (EventDispatchers|CreateEvent :!event Custom|OnBannerRaised)))))

(event Collision|EventActorBeginOverlap
  (params (OtherActor :type Actor%20Object%20Reference)))

(event EventTick
  (params (DeltaSeconds :type Float%20%28single-precision%29)))

(event Custom|OnBannerRaised
  (params (Team :type Integer) (Winners :type Array%20of%20Fort%20Player%20Pawn%20Object%20References) (Banner :type BGA%20POIBanner%20Banner%20Object%20Reference))
  ;[1 1 1] Override in children to do clean up after banner is raised
  (Development|PrintString (Utilities|String|Append "Banner raised by team " (Utilities|String|ToString(Integer) Team))))

(event Custom|OnRep_BannerRef)

(fn ResetAllBannerProgress
  (bind _bannerref (Variables|Default|GetBannerRef))
  (Utilities|IsValid _bannerref
    (:Is%20Valid
      (Progress|ResetAllProgress _bannerref))))

(fn ResetAllBannerProgressWithExclusion
  (params (ExcludeTeams :type Boolean) (TeamsToExclude :type Array%20of%20Integers))
  (bind _bannerref (Variables|Default|GetBannerRef))
  (Utilities|IsValid _bannerref
    (:Is%20Valid
      (Progress|ResetAllProgresswithExclusion _bannerref ExcludeTeams TeamsToExclude))))

(fn ResetBannerProgressForTeam
  (params (Team :type Integer))
  (bind _bannerref (Variables|Default|GetBannerRef))
  (Utilities|IsValid _bannerref
    (:Is%20Valid
      (Progress|ResetTeamProgress _bannerref Team))))

(fn UpdateBannerProgress
  (params (Team :type Integer) (ProgressAmount :type Float%20%28double-precision%29) (bOverrideProgress :type Boolean))
  (bind _bannerref (Variables|Default|GetBannerRef))
  (Utilities|IsValid _bannerref
    (:Is%20Valid
      (Progress|AddProgress _bannerref Team ProgressAmount bOverrideProgress))))

(fn ConstructionScript)
