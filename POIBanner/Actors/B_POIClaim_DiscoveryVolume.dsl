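; Actor Blueprint for a discovery volume in the POI banner claim feature, built around a sphere component.
; On BeginPlay it waits asynchronously for the current playlist's data to be ready.
; It then sets the sphere radius from the Row_DiscoveryVolumeRadius scalable float, without updating overlaps.
; The construction script is empty, and no overlap or discovery logic appears in this graph.
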(event EventBeginPlay
  (Class|FortAsyncActionCurrentPlaylistReady|PlaylistDataReadyAsync
    (:OnReady
      (Components|Sphere|SetSphereRadius (Variables|Default|GetSphere) (ScalableFloat|GetValueatLevel (Variables|Default|GetRow_DiscoveryVolumeRadius)) false))))

(fn ConstructionScript)
