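; Manager actor that tracks which players have claimed each POI through its banner.
; On the server, shortly after playlist data is ready, it gathers all POI banner actors, binds their raised events and indexes them by location tag in a map and a claim-info array.
; A capture writes the winning pawns' player states into that banner's entry, and both setup and capture flush net dormancy.
; Lookup functions return a banner or claim info for a location tag.
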(event EventBeginPlay
  (Class|FortAsyncActionCurrentPlaylistReady|PlaylistDataReadyAsync
    (:OnReady
      (if (Networking|IsServer)
        (Utilities|FlowControl|Delay 0.2)
        (CallFunction|GetAllBanners)))))

(fn GetAllBanners
  (bind _outactors (Actor|GetAllActorsOfClass "/POIBanner/Actors/BGA_POIBanner_Banner.BGA_POIBanner_Banner_C"))
  (bind _output_get (Variables|Default|SetBanners _outactors))
  (for _array_element _output_get
    (Default|BindEventtoOnBannerRaised _array_element (EventDispatchers|CreateEvent :!event OnPOICaptured))
    (Utilities|Map|Add (Variables|Default|GetBannersByLocationTag) (Variables|/POIBanner/Actors/BGA_POIBanner_Banner.BGA_POIBanner_Banner_C|GetPOITag _array_element) _array_element)
    (Utilities|Array|Add (Variables|Default|GetBannerInfo) (Utilities|Struct|MakeFPOIClaimInfo (Variables|/POIBanner/Actors/BGA_POIBanner_Banner.BGA_POIBanner_Banner_C|GetPOITag _array_element) _array_element)))
  (Networking|FlushNetDormancy))

(fn GetBannerForLocation
  (params (LocationTag :type Gameplay%20Tag%20Structure))
  (bind (_value _returnvalue) (Utilities|Map|Find (Variables|Default|GetBannersByLocationTag) LocationTag))
  (return _value _returnvalue))

(fn GetClaimInfoForLocation
  (params (LocationTag :type Gameplay%20Tag%20Structure%20%28by%20ref%29))
  (for _array_element (Variables|Default|GetBannerInfo)
    (if (GameplayTags|MatchesTag (Utilities|Struct|BreakFPOIClaimInfo _array_element) LocationTag true)
      (return true _array_element)))
  (return false 0))

(fn GetPlayerStatesFromWinningPawns
  (params (Pawns :type Array%20of%20Fort%20Player%20Pawn%20Object%20References))
  (for _array_element Pawns
    (bind _asfort_player_state (Utilities|Casting|CastToFortPlayerState (Variables|/Script/Engine.Pawn|GetPlayerState _array_element)))
    (Utilities|Array|Add (LocalVariables|GetOutPlayerStates) _asfort_player_state))
  (return (LocalVariables|GetOutPlayerStates)))

(fn OnPOICaptured
  (params (Team :type Integer) (Winners :type Array%20of%20Fort%20Player%20Pawn%20Object%20References) (Banner :type BGA%20POIBanner%20Banner%20Object%20Reference))
  (bind _bannerinfo (Variables|Default|GetBannerInfo))
  (bind _playerstates (CallFunction|GetPlayerStatesfromWinningPawns Winners))
  (LocalVariables|SetWinnerPlayerStates _playerstates)
  (bind (_array_element _array_index) (Utilities|Array|ForEachLoopwithBreak _bannerinfo)
    (:LoopBody
      (if (== (.Banner_3_148F417641D2F9CA654821923F94F39D (Utilities|Struct|BreakFPOIClaimInfo _array_element)) Banner)
        (Utilities|Array|SetArrayElem _bannerinfo _array_index (Utilities|Struct|MakeFPOIClaimInfo (Utilities|Struct|BreakFPOIClaimInfo _array_element) (.Banner_3_148F417641D2F9CA654821923F94F39D (Utilities|Struct|BreakFPOIClaimInfo _array_element)) (LocalVariables|GetWinnerPlayerStates)))))
    (:Completed
      _))
  (Networking|FlushNetDormancy))

(fn ConstructionScript)
