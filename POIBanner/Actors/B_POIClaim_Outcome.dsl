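; Base outcome actor for POI banner claims, running a response when a team raises its referenced banner.
; BeginPlay and Initialize both bind TryDoOutcome to the banner's OnBannerRaised event.
; If its ScalableFloat enable flag is true, TryDoOutcome wakes net dormancy, calls the empty OnPOIWon hook with team and winners, then returns to full dormancy after a delay.
; POIClaimResetAllProgress only wakes net dormancy.
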(event Custom|POIClaimResetAllProgress
  (Networking|SetNetDormancy "DORM_Awake"))

(event EventBeginPlay
  (bind _bannerref (Variables|Default|GetBannerRef))
  (Utilities|IsValid _bannerref
    (:Is%20Valid
      (Default|BindEventtoOnBannerRaised _bannerref (EventDispatchers|CreateEvent :!event Custom|TryDoOutcome)))))

(event Custom|Initialize
  (bind _bannerref (Variables|Default|GetBannerRef))
  (Utilities|IsValid _bannerref
    (:Is%20Valid
      (Default|BindEventtoOnBannerRaised _bannerref (EventDispatchers|CreateEvent :!event Custom|TryDoOutcome)))))

(event Custom|TryDoOutcome
  (params (Team :type Integer) (Winners :type Array%20of%20Fort%20Player%20Pawn%20Object%20References) (Banner :type BGA%20POIBanner%20Banner%20Object%20Reference))
  (if (CallFunction|IsEnabled)
    (Networking|SetNetDormancy "DORM_Awake")
    (CallFunction|OnPOIWon Team Winners)
    (Utilities|FlowControl|Delay 1e+01)
    (Networking|SetNetDormancy "DORM_DormantAll")))

(event Custom|OnPOIWon
  (params (Team :type Integer) (Winners :type Array%20of%20Fort%20Player%20Pawn%20Object%20References)))

(fn IsEnabled
  (return (ScalableFloat|GetValueasBool (Variables|Default|GetRow_Enabled))))

(fn ConstructionScript)
