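; Player banner actor for POI claiming that starts furled with a neutral icon and displays a player's banner cosmetics.
; OnCaptured replicates a flag index through net dormancy flushes, triggering a timeline-driven material unfurl with random speed and sequential or random delay.
; SetRepresentedPlayer replicates the pawn's banner icon and color IDs.
; Clients then async-load the icon and apply it with the color to a dynamic material instance.
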(event EventBeginPlay
  (CallFunction|InitializeDynamicMaterial)
  (CallFunction|SetUnfurl 0.0)
  (CallFunction|SetShowIcon true)
  (CallFunction|SetBannerIcon (Variables|Default|GetNeutralIcon))
  (if (Variables|Default|GetbApplyNeutralColor)
    (CallFunction|SetBannerColor (Variables|Default|GetNeutralColor))))

(event Custom|OnCaptured
  (params (Index :type Integer))
  (Variables|Default|SetFlagIndex Index)
  (Networking|FlushNetDormancy))

(event Custom|TryUnfurl
  (params (Index :type Integer))
  (bind _bdosequentialunfurl (Variables|Default|GetbDoSequentialUnfurl))
  (bind _returnvalue (CallFunction|GetUnfurlWaitTime _bdosequentialunfurl Index))
  (CallFunction|SetShowIcon true)
  (Components|Timeline|SetPlayRate (Variables|B_POIClaim_PlayerBanner|GetRolloutTL) (CallFunction|GetUnfurlSpeedScalar))
  (if (== _returnvalue 0)
    (CallFunction|DoUnfurl)
    (else
      (Utilities|Time|SetTimerbyEvent (delegate Custom|DoUnfurl) _returnvalue))))

(event Custom|DoUnfurl
  (bind (_direction _newtrack_0 _newtrack_1) (Timeline|RolloutTL PlayFromStart)
    (:Update
      (CallFunction|SetUnfurl _newtrack_0 _newtrack_1))))

(event Custom|StartLoadingBannerAsset
  (params (AssetToLoad :type Soft%20Object%20Reference) (BannerColorName :type Name))
  (bind (_loadedasset _midref _bannercolorname) (BannerAsyncLoadAsset|AsyncLoadBannerAsset AssetToLoad (Variables|Default|GetBannerMat) BannerColorName)
    (:BannerAssetLoadComplete
      (CallFunction|OnBannerAssetLoaded _loadedasset _midref _bannercolorname))))

(event Custom|SetRepresentedPlayer
  (params (Pawn :type Fort%20Pawn%20Object%20Reference))
  (Actor|SetOwner Pawn)
  (bind _asfort_player_pawn_athena (Utilities|Casting|CastToFortPlayerPawnAthena Pawn))
  (Variables|Default|SetBannerIconID (CosmeticLoadout|GetBannerIconId _asfort_player_pawn_athena))
  (Variables|Default|SetBannerColorID (CosmeticLoadout|GetBannerColorId _asfort_player_pawn_athena))
  (Development|PrintString (CosmeticLoadout|GetBannerIconId _asfort_player_pawn_athena))
  (Development|PrintString (CosmeticLoadout|GetBannerColorId _asfort_player_pawn_athena))
  (Variables|Default|SetBannerInfoSet true)
  (Networking|FlushNetDormancy))

(fn AdjustScaleByProgress
  (params (InProgress :type Float%20%28double-precision%29))
  (Transformation|SetWorldScale3D (Variables|Default|GetBannerMesh) (Math|Vector|MakeVector 1.0 1.0 (Math|Float|Lerp (Variables|Default|GetMinScale) (Variables|Default|GetMaxScale) (Math|Curves|GetFloatValue (Variables|Default|GetScaleByProgressCurve) InProgress)))))

(fn GetUnfurlSpeedScalar
  (return (Math|Random|RandomFloatinRange (Variables|Default|GetMinUnfurlSpeedScalar) (Variables|Default|GetMaxUnfurlSpeedScalar))))

(fn GetUnfurlWaitTime
  (params (bPickA :type Boolean) (InInt :type Integer))
  (return (Math|Float|SelectFloat (* (Math|Conversions|ToFloat(Integer) InInt) (Variables|Default|GetSequentialUnfurlTimeDelta)) (Math|Random|RandomFloatinRange 0.0 (Variables|Default|GetMaxRandWait)) bPickA)))

(fn InitializeDynamicMaterial
  (bind _bannermesh (Variables|Default|GetBannerMesh))
  (bind _returnvalue (Rendering|Material|CreateDynamicMaterialInstance _bannermesh 0 (Rendering|Material|GetMaterial _bannermesh 0) "None" :!class /Script/Engine.PrimitiveComponent))
  (bind _output_get (Variables|Default|SetBannerMat _returnvalue))
  (Rendering|Material|SetMaterial _bannermesh 0 _output_get))

(fn OnBannerAssetLoaded
  (params (LoadedAsset :type Object%20Reference) (MID :type Material%20Instance%20Dynamic%20Object%20Reference) (BannerColorName :type Name))
  (bind _bannercolorname (Variables|GetBannerColorName))
  (bind _mid (Variables|GetMID))
  (bind _astexture_2d (Utilities|Casting|CastToTexture2D LoadedAsset))
  (LocalVariables|SetBannerIcon _astexture_2d)
  (Utilities|FlowControl|Sequence
    (:then_0
      (if (== _bannercolorname "")
        (LocalVariables|SetPrimaryColor (Variables|Default|GetDefaultBannerColor))
        (else
          (bind _returnvalue (Banner|GetBannerColorPrimaryColorfromRowName _bannercolorname))
          (LocalVariables|SetPrimaryColor _returnvalue)))
      (Rendering|Material|SetTextureParameterValue _mid (Variables|Default|GetBannerIconMatParam) (LocalVariables|GetBannerIcon))
      (Rendering|Material|SetVectorParameterValue _mid (Variables|Default|GetPrimaryColorMatParam) (LocalVariables|GetPrimaryColor) :!class /Script/Engine.MaterialInstanceDynamic)
      (Rendering|Material|SetVectorParameterValue _mid (Variables|Default|GetSecondaryColorMatParam) (LocalVariables|GetPrimaryColor) :!class /Script/Engine.MaterialInstanceDynamic))))

(fn OnRep_BannerInfo
  (bind _bannerinfo (Variables|Default|GetBannerInfo))
  (bind (_iconid _colorid) (Utilities|Struct|BreakPlayerBannerInfo _bannerinfo))
  (if (not (or (Utilities|String|EqualExactly(String) _iconid) (Utilities|String|EqualExactly(String) _colorid)))
    (CallFunction|InitializeDynamicMaterial)
    (CallFunction|StartCreatingBannerMaterial _bannerinfo)))

(fn OnRep_BannerInfoSet
  (if (Variables|Default|GetBannerInfoSet)
    (if (not (Networking|IsServer))
      (CallFunction|InitializeDynamicMaterial)
      (CallFunction|StartCreatingBannerMaterial :IconID (Utilities|String|SelectString (Variables|Default|GetBannerIconID) (Variables|Default|GetDefaultBannerIconID) (Utilities|String|NotEqualExactly(String) (Variables|Default|GetBannerIconID))) :ColorID (Variables|Default|GetBannerColorID)))))

(fn OnRep_FlagIndex
  (CallFunction|TryUnfurl (Variables|Default|GetFlagIndex)))

(fn SetBannerColor
  (params (Value :type Linear%20Color%20Structure))
  (bind _bannermat (Variables|Default|GetBannerMat))
  (Rendering|Material|SetVectorParameterValue _bannermat (Variables|Default|GetPrimaryColorMatParam) Value :!class /Script/Engine.MaterialInstanceDynamic)
  (Rendering|Material|SetVectorParameterValue _bannermat (Variables|Default|GetSecondaryColorMatParam) Value :!class /Script/Engine.MaterialInstanceDynamic))

(fn SetBannerIcon
  (params (Value :type Texture%20Object%20Reference))
  (Rendering|Material|SetTextureParameterValue (Variables|Default|GetBannerMat) (Variables|Default|GetBannerIconMatParam) Value))

(fn SetShowIcon
  (params (ShouldShow :type Boolean))
  (Rendering|Material|SetScalarParameterValue (Variables|Default|GetBannerMat) "HideIcon" (Math|Float|SelectFloat 0.0 1.0 ShouldShow) :!class /Script/Engine.MaterialInstanceDynamic))

(fn SetUnfurl
  (params (UnfurlValue :type Float%20%28single-precision%29) (WindValue :type Float%20%28double-precision%29))
  (bind _bannermat (Variables|Default|GetBannerMat))
  (Rendering|Material|SetScalarParameterValue _bannermat (Variables|Default|GetUnfurlParamName) UnfurlValue :!class /Script/Engine.MaterialInstanceDynamic)
  (Rendering|Material|SetScalarParameterValue _bannermat (Variables|Default|GetWindParamName) WindValue :!class /Script/Engine.MaterialInstanceDynamic))

(fn SetVisualsFromState
  (params (State :type EPOIBannerState%20Enum :default "NewEnumerator0")))

(fn StartCreatingBannerMaterial
  (params (BannerInfo :type Player%20Banner%20Info%20Structure) (IconID :type String) (ColorID :type String))
  (bind _returnvalue (Banner|GetBannerLargeIconAssetfromRowName (Utilities|String|StringToName IconID) "PhysicalBanner"))
  (CallFunction|StartLoadingBannerAsset _returnvalue (Utilities|String|StringToName ColorID)))

(fn ConstructionScript)
