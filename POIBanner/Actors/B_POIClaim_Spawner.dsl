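; Placeable POI Claim spawner that sets up a banner capture challenge at its location.
; Once playlist data is ready, it destroys itself if disabled; otherwise the server line-traces vertically for a surface.
; On a hit it spawns and links a banner, capture point, discovery volume and outcome actors, all destroyed by the server on EndPlay.
; In editor, the construction script scales a preview mesh to the data-driven or overridden capture radius.
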(event EventBeginPlay
  (Class|FortAsyncActionCurrentPlaylistReady|PlaylistDataReadyAsync
    (:OnReady
      (Utilities|FlowControl|Sequence
        (:then_0
          (if (ScalableFloat|GetValueasBool (Variables|Default|GetRow_Enabled))
            (CallFunction|SpawnChallenge)
            (else
              (Actor|DestroyActor))))
        (:then_1
          (if (Editor|InEditororPIE)
            (Components|DestroyComponent (Variables|Default|GetPreviewMesh))))))))

(event EventEndPlay
  (params (EndPlayReason :type EEndPlayReason%20Enum :default "Destroyed"))
  ;[0.1354 0.07906 0.01371] EndPlay: Clean up spawned actors.
  (|CallParentFunction EndPlayReason)
  (Networking|SwitchHasAuthority
    (:Authority
      (CallFunction|CleanupSpawnedActors))))

(fn CleanupSpawnedActors
  (bind _actorstodestroy (LocalVariables|GetActorsToDestroy))
  ;[0.1094 0.02868 0.1086] CleanupSpawnedActors: Destroy all the actors that were spawned during this actor's setup.
  (Networking|SwitchHasAuthority
    (:Authority
      (Utilities|FlowControl|Sequence
        (:then_0
          (Utilities|Array|Add _actorstodestroy (Variables|SpawnedActors|GetBanner))
          (Utilities|Array|AppendArray _actorstodestroy (Variables|SpawnedActors|GetControlPoints))
          (Utilities|Array|Add _actorstodestroy (Variables|SpawnedActors|GetDiscoveryVolume))
          (Utilities|Array|AppendArray _actorstodestroy (Variables|SpawnedActors|GetOutcomeActors)))
        (:then_1
          (for _array_element (LocalVariables|GetActorsToDestroy)
            (Utilities|IsValid _array_element
              (:Is%20Valid
                (Actor|DestroyActor _array_element)))))))))

(fn FindSpawnPoint
  (bind _outhit (Collision|LineTraceByChannel (CallFunction|GetTraceStartLocation) (CallFunction|GetTraceEndLocation) "TraceTypeQuery7"))
  (return (Collision|BreakHitResult _outhit) (Math|Vector|SelectVector (.Location (Collision|BreakHitResult _outhit)) (.TraceEnd (Collision|BreakHitResult _outhit)) (Collision|BreakHitResult _outhit))))

(fn GetTraceEndLocation
  (return (- (Transformation|GetActorLocation) (Variables|Default|GetTraceDownOffset))))

(fn GetTraceStartLocation
  (return (+ (Transformation|GetActorLocation) (Variables|Default|GetTraceUpOffset))))

(fn SpawnChallenge
  (Networking|SwitchHasAuthority
    (:Authority
      (bind (_bblockinghit _location) (CallFunction|FindSpawnPoint))
      (if _bblockinghit
        (CallFunction|SpawnControlPointActors _location)))))

(fn SpawnControlPointActors
  (params (Location :type Vector))
  (LocalVariables|SetSpawnTransform (Math|Conversions|ToTransform(Vector) Location))
  (Utilities|FlowControl|Sequence
    (:then_0
      ;[1 1 1] Spawn main banner
      (bind _returnvalue (Building|SpawnBuildingGameplayActor (Variables|Default|GetBannerClass) (LocalVariables|GetSpawnTransform) self true))
      (bind _asbga_poibanner_banner (Utilities|Casting|CastToBGA_POIBanner_Banner _returnvalue))
      (Variables|SpawnedActors|SetBanner _asbga_poibanner_banner)
      (Variables|/POIBanner/Actors/BGA_POIBanner_Banner.BGA_POIBanner_Banner_C|SetPOITag (Variables|Default|GetPOITag) _asbga_poibanner_banner))
    (:then_1
      ;[1 1 1] Spawn capture point
      (bind _returnvalue_2 (Game|SpawnActorfromClass (Variables|Default|GetControlPointClass) (LocalVariables|GetSpawnTransform) "Undefined" "SelectDefaultAtRuntime" self (Utilities|Array|MakeArray) (Variables|Default|GetShould%20Override%20Radius) (Variables|Default|GetRadius%20Override) (Variables|SpawnedActors|GetBanner) 0 0))
      (Utilities|Array|Add (Variables|/POIBanner/Actors/BGA_POIBanner_Banner.BGA_POIBanner_Banner_C|GetProgressSources (Variables|SpawnedActors|GetBanner)) _returnvalue_2)
      (Variables|/POIBanner/Actors/CapturePoint/BGA_POIBanner_CapturePoint.BGA_POIBanner_CapturePoint_C|SetOptOutOfCaptureHeightCheck (Variables|Default|GetOptOutOfCaptureHeightCheck) _returnvalue_2)
      (Utilities|Array|Add (Variables|SpawnedActors|GetControlPoints) _returnvalue_2))
    (:then_2
      ;[1 1 1] Spawn discovery volume
      (bind _returnvalue_3 (Game|SpawnActorfromClass (Variables|Default|GetDiscoveryVolumeClass) (LocalVariables|GetSpawnTransform) "AlwaysSpawn" "SelectDefaultAtRuntime" 0 0))
      (Variables|SpawnedActors|SetDiscoveryVolume _returnvalue_3)
      (bind _asb_poiclaim_discovery_volume (Utilities|Casting|CastToB_POIClaim_DiscoveryVolume _returnvalue_3))
      (Variables|/POIBanner/Actors/B_POIClaim_DiscoveryVolume.B_POIClaim_DiscoveryVolume_C|SetBanner (Variables|SpawnedActors|GetBanner) _asb_poiclaim_discovery_volume))
    (:then_3
      ;[1 1 1] Spawn outcome handlers
      (for _array_element (Variables|Default|GetOutcomeClasses)
        (bind _returnvalue_4 (Game|SpawnActorfromClass _array_element (LocalVariables|GetSpawnTransform) "Undefined" "SelectDefaultAtRuntime" self (Variables|SpawnedActors|GetBanner) 0))
        (Utilities|Array|Add (Variables|SpawnedActors|GetOutcomeActors) _returnvalue_4)))))

(fn ConstructionScript
  (bind (_returnvalue_x _returnvalue_y _returnvalue_z) (Math|Vector|BreakVector (Transformation|GetWorldScale (Variables|Default|GetPreviewMesh))))
  (if (Editor|InEditor)
    ;[1 1 1] This is jank but should keep the preview mesh scale accurate to the current capture radius from data
    (Transformation|SetWorldScale3D (Variables|Default|GetPreviewMesh) (Math|Vector|MakeVector (/ (Math|Float|SelectFloat (Variables|Default|GetRadius%20Override) (ScalableFloat|GetValueatLevel (Variables|Default|GetRow_CaptureRadius)) (Variables|Default|GetShould%20Override%20Radius)) (Variables|Default|GetPreviewMeshScaleConversion)) (/ (Math|Float|SelectFloat (Variables|Default|GetRadius%20Override) (ScalableFloat|GetValueatLevel (Variables|Default|GetRow_CaptureRadius)) (Variables|Default|GetShould%20Override%20Radius)) (Variables|Default|GetPreviewMeshScaleConversion)) _returnvalue_z))))
